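# Conversation database for the Scrapknight Jules NPC, defining what she says and which shared menus she offers.
# A single entry point leads to a speech node with a default greeting plus prioritized contextual lines for a player's first-ever conversation or specific worn outfits.
# The speech node then branches through link nodes into the shared NPC library quests and basic services conversations by entry tag.
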ConversationDatabase /NPCLibrary/NPCs/ScrapknightJules/NPCLibrary_Conversation_ScrapknightJules
entry_points [{tag=Conversation.Entry.ScrapknightJules, destinations [#1]}]
nodes [
 ConversationEntryPointNode_7 {
  id=#1
  class=ConversationEntryPointNode
  outputs [#2]
  properties {entryTag {tagName=Conversation.Entry.ScrapknightJules}}
 }
 Speech_C_4 {
  id=#2
  class=Speech_C
  outputs [#3, #4]
  properties {
   nodeName=""
   generalConfig {
    defaultMessage=No looking back. Only forward.
    contextualMessages [
     {
      message=This kingdom's got neat tech. Can't wait to rip it apart and build it better.
      contextRequirements [{requirement {class=FortControllerRequirement_IsFirstEverConversationWithNPC}}]
      priority {curve {rowName=Tandem.Speech.Priority.FirstEverInteraction}}
     }
     {
      message=I could probably throw together some better tech for vampire slaying.
      contextRequirements [{requirement {class=FortControllerRequirement_HasCID, properties {allowedCIds [&Character_SharpFang]}}}]
      priority {curve {rowName=Tandem.Speech.Priority.OutfitReaction}}
     }
     {
      message=How about some jet boosters for that board?
      contextRequirements [{requirement {class=FortControllerRequirement_HasCID, properties {allowedCIds [/BRCosmetics/Athena/Items/Cosmetics/Characters/Character_Venice]}}}]
      priority {curve {rowName=Tandem.Speech.Priority.OutfitReaction}}
     }
     {
      message="If you need a psychic suppressor, I could rig something up."
      contextRequirements [{requirement {class=FortControllerRequirement_HasCID, properties {allowedCIds [/BRCosmetics/Athena/Items/Cosmetics/Characters/Character_Apprentice]}}}]
      priority {curve {rowName=Tandem.Speech.Priority.OutfitReaction}}
     }
     {
      message=That armor is impressive. What's it made of?
      contextRequirements [{requirement {class=FortControllerRequirement_HasCID, properties {allowedCIds [&Character_Sunlit]}}}]
      priority {curve {rowName=Tandem.Speech.Priority.OutfitReaction}}
     }
     {
      message="It's an honour to be working with your Rift Warden, your majesty."
      contextRequirements [{requirement {class=FortControllerRequirement_HasCID, properties {allowedCIds [&Character_Citadel]}}}]
      priority {curve {rowName=Tandem.Speech.Priority.OutfitReaction}}
     }
     {
      message=Who made your armor?
      contextRequirements [{requirement {class=FortControllerRequirement_HasCID, properties {allowedCIds [/BRCosmetics/Athena/Items/Cosmetics/Characters/Character_RedPepper]}}}]
      priority {curve {rowName=Tandem.Speech.Priority.OutfitReaction}}
     }
     {
      message=What's the deal with the two swords?
      contextRequirements [{requirement {class=FortControllerRequirement_HasCID, properties {allowedCIds [/BRCosmetics/Athena/Items/Cosmetics/Characters/Character_MagicMeadow]}}}]
      priority {curve {rowName=Tandem.Speech.Priority.OutfitReaction}}
     }
     {
      message=I'll help you. For the island.
      contextRequirements [{requirement {
         class=FortControllerRequirement_HasCID
         properties {allowedCIds [&CID_A_425_Athena_Commando_F_BlueJay]}
        }}]
      priority {curve {rowName=Tandem.Speech.Priority.OutfitReaction}}
     }
     {
      message=I don't want to get pulled into any schemes.
      requirementMatchPolicy=RequireAny
      contextRequirements [{requirement {
         class=FortControllerRequirement_HasCID
         properties {allowedCIds [
           /BRCosmetics/Athena/Items/Cosmetics/Characters/CID_A_129_Athena_Commando_M_CatBurglarSummer
           /BRCosmetics/Athena/Items/Cosmetics/Characters/CID_A_257_Athena_Commando_M_CatBurglar_Ghost
           /BRCosmetics/Athena/Items/Cosmetics/Characters/CID_A_320_Athena_Commando_M_CatburglarWinter
           /BRCosmetics/Athena/Items/Cosmetics/Characters/CID_A_390_Athena_Commando_M_Blackbird
           /BRCosmetics/Athena/Items/Cosmetics/Characters/CID_694_Athena_Commando_M_CatBurglar
           /BRCosmetics/Athena/Items/Cosmetics/Characters/CID_923_Athena_Commando_M_ParcelGold
          ]}
        }}]
      priority {curve {rowName=Tandem.Speech.Priority.OutfitReaction}}
     }
    ]
   }
  }
 }
 ConversationLinkNode_0 {id=#4, class=ConversationLinkNode, properties {remoteEntryTag {tagName=Conversation.Entry.BasicServices}}}
 ConversationLinkNode_1 {id=#3, class=ConversationLinkNode, properties {remoteEntryTag {tagName=Conversation.Entry.NPCLibrary.Quests}}}
]
